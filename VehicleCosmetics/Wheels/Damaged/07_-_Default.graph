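# Material for a damaged vehicle wheel cosmetic, variant 07 Default, going by its path and name.
# It uses Default Lit shading and is flagged for use on skeletal meshes and with mesh deformers.
# The graph holds a single vector parameter with a gold default color, and shows no connections to material outputs, textures or damage logic.
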Material /VehicleCosmetics/Wheels/Damaged/07_-_Default
graph {
 properties {bUsedWithMeshDeformer=true, bUsedWithSkeletalMesh=true, pixelDepthOffsetMode=PDOM_Legacy, usedShadingModels=Default Lit}
 nodes [VectorParameter() {defaultValue=<rgba 0.8118 0.6863 0 0>}]
}
